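# Sound cue for the end-of-reload sound of the Dragon's Breath shotgun.
# A single wave player runs through a modulator that fixes pitch and slightly lowers volume, with no random variation.
# It uses the shared weapon reload attenuation and weapon foley sound class.
# A concurrency override limits it to a couple of instances per owner, stopping the oldest.
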SoundCue /Game/Sounds/Fort_Weapons/Named_Weapons/Dragons_Breath_SG/WPN_SG_DragonsBreath_Reload_End
root=Modulator_0
nodes [
 {id=Modulator_0, children [WavePlayer_1], properties {pitchMax=1, pitchMin=1, volumeMax=0.8, volumeMin=0.8}}
 {
  id=WavePlayer_1
  properties {soundWaveAssetPtr=&/WPN_Shotgun_DragonsBreath_Reload_End_01}
 }
]
properties {
 attenuationSettings=&Fort_Weapon_Reload_Att
 bOverrideConcurrency=true
 maxDistance=2250
 soundClassObject=&Fort_WeaponFoley
 volumeMultiplier=0.501
 concurrencyOverrides {bLimitToOwner=true, maxCount=2, resolutionRule=StopOldest}
}
