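# Sound cue for the shell-insert reload sound of the Dragon's Breath shotgun.
# A mixer layers a random generic shotgun shell-insert wave, at lower volume, with one of four random, equally weighted Dragon's Breath-specific insert waves.
# It uses weapon foley class and reload attenuation, with a concurrency override that prevents new instances once the limit is reached.
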SoundCue /Game/Sounds/Fort_Weapons/Named_Weapons/Dragons_Breath_SG/WPN_SG_DragonsBreath_Reload_Insert
root=Mixer_1
nodes [
 {id=Mixer_1, children [Random_2,Random_0], properties {inputVolume [0.35,0.65]}}
 {id=Random_0, children [WavePlayer_2,WavePlayer_3,WavePlayer_4,WavePlayer_5], properties {weights [1,1,1,1]}}
 {id=Random_2, children [WavePlayer_0,WavePlayer_1], properties {weights [1,1]}}
 {id=WavePlayer_0, properties {soundWaveAssetPtr=/Game/Sounds/Fort_Weapons/Shotgun/Reload_Elements/weapon_shotgun_insert_shell_03}}
 {id=WavePlayer_1, properties {soundWaveAssetPtr=/Game/Sounds/Fort_Weapons/Shotgun/Reload_Elements/weapon_shotgun_insert_shell_04}}
 {
  id=WavePlayer_2
  properties {soundWaveAssetPtr=&/WPN_Shotgun_DragonsBreath_Insert_01}
 }
 {
  id=WavePlayer_3
  properties {soundWaveAssetPtr=&/WPN_Shotgun_DragonsBreath_Insert_03}
 }
 {
  id=WavePlayer_4
  properties {soundWaveAssetPtr=&/WPN_Shotgun_DragonsBreath_Insert_04}
 }
 {
  id=WavePlayer_5
  properties {soundWaveAssetPtr=&/WPN_Shotgun_DragonsBreath_Insert_02}
 }
]
properties {
 attenuationSettings=&Fort_Weapon_Reload_Att
 bOverrideConcurrency=true
 maxDistance=2250
 soundClassObject=&Fort_WeaponFoley
 volumeMultiplier=1.103
 concurrencyOverrides {resolutionRule=PreventNew}
}
