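# Sound cue for the start of the Dragon's Breath shotgun reload, routed as weapon foley with reload attenuation.
# A mixer layers the main reload-start wave, slightly delayed, with four later layers of shell sounds.
# Each shell layer randomly picks one of three shared break-shell waves, with its own delay, pitch and volume modulation, so the staggered clatter varies per play.
# A concurrency override prevents new instances once the limit is reached.
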SoundCue /Game/Sounds/Fort_Weapons/Named_Weapons/Dragons_Breath_SG/WPN_SG_DragonsBreath_Reload_Start
root=Mixer_0
nodes [
 {id=Delay_11, children [Modulator_2], properties {delayMax=0.95, delayMin=0.95}}
 {id=Delay_12, children [Modulator_3], properties {delayMax=1.3, delayMin=1.2}}
 {id=Delay_13, children [Modulator_4], properties {delayMax=1.4, delayMin=1.4}}
 {id=Delay_2, children [Mixer_1], properties {delayMax=0.18, delayMin=0.18}}
 {id=Delay_9, children [Modulator_1], properties {delayMax=1.3, delayMin=1.2}}
 {id=Mixer_0, children [Modulator_0,Delay_11,Delay_9,Delay_12,Delay_13], properties {inputVolume [1.1,1,1,1,1]}}
 {id=Mixer_1, children [WavePlayer_1], properties {inputVolume [1]}}
 {id=Modulator_0, children [Delay_2], properties {pitchMax=1, pitchMin=1, volumeMax=1, volumeMin=1}}
 {id=Modulator_1, children [Random_0], properties {pitchMax=1.3, pitchMin=1, volumeMax=1, volumeMin=0.8}}
 {id=Modulator_2, children [Random_3], properties {pitchMax=1.3, pitchMin=1, volumeMax=1, volumeMin=0.8}}
 {id=Modulator_3, children [Random_1], properties {pitchMax=1, pitchMin=0.8, volumeMax=1, volumeMin=0.8}}
 {id=Modulator_4, children [Random_2], properties {pitchMax=1.3, pitchMin=1.1, volumeMax=0.7, volumeMin=0.6}}
 {id=Random_0, children [WavePlayer_0,WavePlayer_2,WavePlayer_3], properties {weights [1,1,1]}}
 {id=Random_1, children [WavePlayer_4,WavePlayer_5,WavePlayer_6], properties {weights [1,1,1]}}
 {id=Random_2, children [WavePlayer_7,WavePlayer_8,WavePlayer_9], properties {weights [1,1,1]}}
 {id=Random_3, children [WavePlayer_12, WavePlayer_13, WavePlayer_14], properties {weights [1,1,1]}}
 {id=WavePlayer_0, properties {soundWaveAssetPtr=/Game/Sounds/Fort_Weapons/Shotgun/Reload_Elements/weapon_shotgun_break_shells_01}}
 {
  id=WavePlayer_1
  properties {soundWaveAssetPtr=&/WPN_Shotgun_DragonsBreath_Reload_Start}
 }
 {id=WavePlayer_12, properties {soundWaveAssetPtr=&weapon_shotgun_break_shells_01}}
 {id=WavePlayer_13, properties {soundWaveAssetPtr=/Game/Sounds/Fort_Weapons/Shotgun/Reload_Elements/weapon_shotgun_break_shells_02}}
 {id=WavePlayer_14, properties {soundWaveAssetPtr=/Game/Sounds/Fort_Weapons/Shotgun/Reload_Elements/weapon_shotgun_break_shells_03}}
 {id=WavePlayer_2, properties {soundWaveAssetPtr=&weapon_shotgun_break_shells_02}}
 {id=WavePlayer_3, properties {soundWaveAssetPtr=&weapon_shotgun_break_shells_03}}
 {id=WavePlayer_4, properties {soundWaveAssetPtr=&weapon_shotgun_break_shells_01}}
 {id=WavePlayer_5, properties {soundWaveAssetPtr=&weapon_shotgun_break_shells_02}}
 {id=WavePlayer_6, properties {soundWaveAssetPtr=&weapon_shotgun_break_shells_03}}
 {id=WavePlayer_7, properties {soundWaveAssetPtr=&weapon_shotgun_break_shells_01}}
 {id=WavePlayer_8, properties {soundWaveAssetPtr=&weapon_shotgun_break_shells_02}}
 {id=WavePlayer_9, properties {soundWaveAssetPtr=&weapon_shotgun_break_shells_03}}
]
properties {
 attenuationSettings=&Fort_Weapon_Reload_Att
 bOverrideConcurrency=true
 maxDistance=2250
 soundClassObject=&Fort_WeaponFoley
 volumeMultiplier=0.621
 concurrencyOverrides {resolutionRule=PreventNew}
}
