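# Sound cue for the first-person firing sound of the Dragon's Breath shotgun.
# A mixer plays one of four weapon-specific shot variants at full volume over a much quieter secondary layer.
# That layer picks one of three generic bigger-shotgun waves, pitched up by a fixed amount and slightly delayed behind the main shot.
# First-person gunfire attenuation, sound class and concurrency assets govern playback.
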SoundCue /Game/Sounds/Fort_Weapons/Named_Weapons/Dragons_Breath_SG/WPN_Shotgun_DragonsBreath_Fire_1P
root=Mixer_0
nodes [
 {id=Delay_0, children [Modulator_0], properties {delayMax=0.12, delayMin=0.12}}
 {id=Mixer_0, children [Random_1,Delay_0], properties {inputVolume [1,0.2]}}
 {id=Modulator_0, children [Random_2], properties {pitchMax=1.2, pitchMin=1.2}}
 {id=Random_1, children [WavePlayer_0,WavePlayer_5,WavePlayer_6,WavePlayer_7], properties {weights [1,1,1,1]}}
 {id=Random_2, children [WavePlayer_8, WavePlayer_9, WavePlayer_10], properties {weights [1,1,1]}}
 {
  id=WavePlayer_0
  properties {soundWaveAssetPtr=&/WPN_Shotgun_DragonsBreath_Fire_1P_v2_01}
 }
 {id=WavePlayer_10, properties {soundWaveAssetPtr=/Game/Sounds/Fort_Weapons/Shotgun/weapons_shotguns_bigger_015}}
 {
  id=WavePlayer_5
  properties {soundWaveAssetPtr=&/WPN_Shotgun_DragonsBreath_Fire_1P_v2_02}
 }
 {
  id=WavePlayer_6
  properties {soundWaveAssetPtr=&/WPN_Shotgun_DragonsBreath_Fire_1P_v2_03}
 }
 {
  id=WavePlayer_7
  properties {soundWaveAssetPtr=&/WPN_Shotgun_DragonsBreath_Fire_1P_v2_04}
 }
 {id=WavePlayer_8, properties {soundWaveAssetPtr=/Game/Sounds/Fort_Weapons/Shotgun/weapons_shotguns_bigger_013}}
 {id=WavePlayer_9, properties {soundWaveAssetPtr=/Game/Sounds/Fort_Weapons/Shotgun/weapons_shotguns_bigger_014}}
]
properties {
 attenuationSettings=&Fort_Weapon_Projectile_FirstPerson_Athena_Att
 maxDistance=2097152
 soundClassObject=&Fort_WeaponRanged_Fire_1P_PSM
 volumeMultiplier=1
 attenuationOverrides {falloffDistance=5000, attenuationShapeExtents {x=1000}}
 concurrencySet [&Weapon_Gunfire_FirstPerson_Conc]
}
